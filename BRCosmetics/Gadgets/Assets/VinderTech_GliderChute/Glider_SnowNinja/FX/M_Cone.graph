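# Unlit, two-sided translucent particle material for a cone effect on the SnowNinja glider, colored by particle color with a camera offset function.
# Opacity multiplies a vertical gradient by two summed, panning caustic and noise textures, sharpened by a power.
# A dynamic parameter can drive pan time and scales alpha.
# Edge-on faces fade via camera-normal dot product, and depth fades soften intersections, with the final one skipped on mobile feature levels.
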Material /BRCosmetics/Gadgets/Assets/VinderTech_GliderChute/Glider_SnowNinja/FX/M_Cone
graph {
 properties {
  bUsedWithMeshParticles=true
  bUsedWithNiagaraMeshParticles=true
  bUsedWithParticleSprites=true
  blendMode=BLEND_Translucent
  pixelDepthOffsetMode=PDOM_Legacy
  shadingModel=MSM_Unlit
  twoSided=true
  usedShadingModels=Unlit
 }
 nodes [
  Add(
   A=TextureSample() {texture=/Game/Effects/Fort_Effects/Textures/Ramps/T_Ramp_Radial_Soft}.R
   B=TextureSample() {samplerType=SAMPLERTYPE_Grayscale, texture=&T_Smoke_Trail_Soft}.R
  )
  n1 = ParticleColor()
  n2 = TextureCoordinate()
  n3 = Multiply(A=OneMinus(None=ComponentMask(None=@n2) {g=true})) {constB=2}
  n4 = TextureCoordinate() {vTiling=0.3}
  n5 = LinearInterpolate(A=Time(), B=DynamicParameter() {paramNames [Time,Param2,Param3,Alpha]}.Time, Alpha=ScalarParameter() {parameterName=UseDPasTime})
  n6 = Add(
   A=TextureSample(UVs=Panner(Coordinate=@n4, Time=@n5) {speedY=-1}) {texture=&T_ShieldCaustics}.R
   B=TextureSample(UVs=Panner(Coordinate=@n4, Time=@n5) {speedY=-0.5}) {samplerType=SAMPLERTYPE_LinearColor, texture=&T_NoisePack_LinearColor}.R
  )
  n7 = Multiply(A=@n1.A, B=Power(Base=Multiply(A=@n3, B=@n6), Exp=ScalarParameter() {defaultValue=3, parameterName=Exp}))
  n8 = ScalarParameter() {defaultValue=5, parameterName=OpacityDF}
  n9 = Abs(None=DotProduct(A=CameraVectorWS(), B=VertexNormalWS()))
  n10 = Multiply(A=@n9, B=@n9)
  n11 = Multiply(
   A=Multiply(A=DepthFade(Opacity=@n7, FadeDistance=@n8), B=Multiply(A=@n10, B=@n10))
   B=DynamicParameter() {paramNames [Time,Param2,Param3,Alpha]}.Alpha
  )
  n12 = DepthFade(Opacity=@n11, FadeDistance=ScalarParameter() {defaultValue=5, parameterName=DepthFade})
  FeatureLevelSwitch(Default=@n12, ES2=@n11, ES3_1=@n11, SM5=@n12, SM6=@n12)
  MaterialFunctionCall(XY=@n2) {materialFunction=&Swizzle}
  MaterialFunctionCall(Offset Amount=ScalarParameter() {defaultValue=10, parameterName=CameraOffset}) {materialFunction=/Engine/Functions/Engine_MaterialFunctions02/WorldPositionOffset/CameraOffset}
  MaterialFunctionCall(Emissive=@n1.RGB) {materialFunction=&MF_FX_EmissiveAdjustments}
 ]
}
