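# Translucent, two-sided lit mesh-particle material for wave effects, evidently shown when the Snow Ninja glider deploys.
# Two noise textures pan at different scales and speeds, driven by the dynamic Time parameter, then multiply and square-root into a turbulence pattern.
# That pattern, tinted by particle color, feeds emissive adjustments.
# It also builds an alpha mask with squared view-angle fade, sine UV falloffs, an Offset-driven wipe and particle alpha.
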Material /BRCosmetics/Gadgets/Assets/VinderTech_GliderChute/Glider_SnowNinja/FX/M_DeployWaves
graph {
 properties {
  bUsedWithMeshParticles=true
  bUsedWithNiagaraMeshParticles=true
  blendMode=BLEND_Translucent
  pixelDepthOffsetMode=PDOM_Legacy
  previewMesh=/BRCosmetics/Effects/Fort_Effects/Mesh/Tornado/S_RoundWave
  translucencyLightingMode=TLM_VolumetricDirectional
  twoSided=true
  usedShadingModels=Default Lit
 }
 nodes [
  n1 = VectorParameter() {parameterName=UV, defaultValue=<rgba 0.5 0.5 1 -0.5>}
  n2 = Multiply(A=TextureCoordinate(), B=AppendVector(A=@n1.R, B=@n1.R))
  n3 = DynamicParameter() {defaultValue=<rgba 1 1 0 1>, paramNames [Time,Random,Offset,P4]}
  n4 = Add(A=@n3.Time, B=@n3.Time)
  n5 = AppendVector(A=@n1.B, B=@n1.B)
  n6 = Panner(Coordinate=Multiply(A=@n2) {constB=0.618}, Time=@n4, Speed=Multiply(A=@n5) {constB=1.618})
  n7 = Panner(Coordinate=@n2, Time=@n4, Speed=@n5)
  n8 = SquareRoot(None=Multiply(
    A=TextureSample(UVs=@n6) {texture=/Game/Global/Textures/Default/Caustics/T_Turbulence_Seamless}.G
    B=TextureSample(UVs=@n7) {samplerType=SAMPLERTYPE_LinearColor, texture=&T_NoisePack_LinearColor}.R
   ))
  n9 = ParticleColor()
  MaterialFunctionCall(Emissive=Multiply(A=Add(A=@n8) {constB=0.5}, B=@n9.RGB)) {materialFunction=&MF_FX_EmissiveAdjustments}
  n10 = DotProduct(A=VertexNormalWS(), B=CameraVectorWS())
  n11 = Multiply(A=Multiply(A=@n8) {constB=2}, B=Multiply(A=@n10, B=@n10))
  n12 = MaterialFunctionCall(Float2=Sine(None=TextureCoordinate() {uTiling=0.5, vTiling=0.5})) {materialFunction=&BreakOutFloat2Components}
  n13 = Multiply(A=@n12.R, B=@n12.R)
  n14 = TextureCoordinate() {uTiling=0.7, vTiling=0.7}
  n15 = Add(A=OneMinus(None=Multiply(A=@n14, B=@n14)), B=DynamicParameter() {defaultValue=<rgba 1 1 0 1>, paramNames [Time,Random,Offset,P4]}.Offset)
  n16 = Sine(None=ComponentMask(None=Saturate(None=@n15)) {r=true})
  Multiply(A=@n9.A, B=Multiply(A=Multiply(A=@n11, B=@n13), B=Saturate(None=@n16)))
  TextureSample() {samplerType=SAMPLERTYPE_LinearColor, texture=/Game/Global/Textures/Default/Tileable_Patterns/T_Tile_Smoke_Small}
  TextureSample()
  TextureSample(UVs=@n6) {samplerType=SAMPLERTYPE_Normal, texture=/Game/Effects/Fort_Effects/Textures/Plasma/T_Turbulence_Seamless_N}
 ]
}
