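# Niagara deploy effect for the SnowNinja glider, playing delayed air-swipe meshes on both sides of the glider.
# Two near-identical, slightly staggered emitters each burst two partition-varied particles, sampling the user skeletal mesh with opposite X offsets during update.
# Curves drive stretching mesh scale, dynamic material parameters and alpha fade, with a second alpha scale linked to time of day.
# Front-end modifiers switch delay, offset and brightness between gameplay and menus.
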NiagaraSystem /BRCosmetics/Gadgets/Assets/VinderTech_GliderChute/Glider_SnowNinja/FX/NS_Glider_RankedRemixAsparagus_Deploy
dependencies {usedDataInterfaces [&NiagaraDataInterfaceCurve], usedDynamicInputs [], usedModules [&SystemState]}
emitters {
 AirSwipe_R001 {
  emitterUpdateScript [
   &EmitterState(
    Life Cycle Mode=::NewEnumerator1
    Loop Behavior=::NewEnumerator1
    Loop Delay=&FN_FrontEndFloatModifier(Gameplay=1.5, Front End=2)
    Loop Duration=1
    UseLoopDelay=true
   )
   &SpawnBurst_Instantaneous(Spawn Count=2, Use Spawn Probability=false, UseLoopCountLimit=false)
  ]
  particleSpawnScript [
   &PartitionParticles()
   &InitializeParticle(
    Color Mode=::NewEnumerator0
    Lifetime=&Lerp_Float(A=0.35, B=0.8, Alpha=&MakeFloatFromInt(Int=$Particles.PartitionParticles.Partition))
    Mesh Scale Mode=::NewEnumerator3
    Mesh Uniform Scale=&Lerp_Float(A=0.6, B=0.8, Alpha=&MakeFloatFromInt(Int=$Particles.PartitionParticles.Partition))
    UsePositionOffset=false
    Write Material Random=true
   )
   &AddVelocity(Velocity=<150 -75 0>)
   &InitialMeshOrientation(
    Mesh Orientation Mode=::NewEnumerator7
    Rotation=&Lerp_Vector(A=<0 0 -0.2>, B=<0 0 0>, Alpha=&MakeFloatFromInt(Int=$Particles.PartitionParticles.Partition))
    Use Rotation Vector=true
   )
  ]
  particleUpdateScript [
   &ParticleState()
   &SkeletalMeshLocation(
    Bone / Socket Index=2
    MaskSpawnGroup=false
    Mesh Sampling Type=::NewEnumerator5
    Sampled Position Offset=&FN_FrontEndVector3Modifier(Gameplay=<60 0 -10>, Front End=<60 0 -50>)
    Skeletal Mesh=$User.Skeletal Mesh
    Socket Sampling Mode=::NewEnumerator5
   )
   &Drag(Drag=1.5, Use Linear Drag=true, Use Rotational Drag=true)
   &ScaleMeshSize(Scale Factor=/Niagara/DynamicInputs/CascadeConversion/SplitVector(
     Double Component Vector Channels:ECascadeNiagaraTwoVectorChannels=::NewEnumerator0
     Double Component:NiagaraFloat=&FloatFromCurve(
      FloatCurve=DataInterface() {propertyValues {Curve {keys [
          {arriveTangent=0.0236, interpMode=RCIM_Cubic, time=0, value=0.1989}
          {arriveTangent=1.0027, interpMode=RCIM_Cubic, leaveTangent=1.0027, time=0.2593, value=0.8027}
          <at 1 1.2016 RCIM_Cubic>
         ]}}}
      Scale Curve=0.7
     )
     Single Component:NiagaraFloat=&FloatFromCurve(
      FloatCurve=DataInterface() {propertyValues {Curve {keys [<at 0 1 RCIM_Cubic>, {arriveTangent=-1, interpMode=RCIM_Cubic, leaveTangent=-1, time=0.2593, value=1}, <at 1 0 RCIM_Cubic>]}}}
      Scale Curve=2
     )
    ))
   &SolveForcesAndVelocity(Clamp Velocity=false, Limit Acceleration=false)
   &DynamicMaterialParameters(
    Index 0 Param 1=&FloatFromCurve(FloatCurve=DataInterface() {propertyValues {Curve {keys [
         <at 0 -1.9537 RCIM_Cubic>
         {arriveTangent=0.4535, interpMode=RCIM_Cubic, leaveTangent=0.4535, time=0.413, value=-1.7083}
         {interpMode=RCIM_Cubic, leaveTangent=0.1229, time=1, value=-1.5002}
        ]}}})
    Index 0 Param 3=&FloatFromCurve(FloatCurve=DataInterface() {propertyValues {Curve {keys [{arriveTangent=-2.0245, interpMode=RCIM_Cubic, time=0, value=0.004}, <at 1 -0.8019 RCIM_Cubic>]}}})
    Index 0 Param 4=1
    Module.Index0Param0Write=true
    Module.Index0Param1Write=false
    Module.Index0Param2Write=true
    Module.Index0Param3Write=true
   )
   &ScaleColor(
    Scale Alpha=&FloatFromCurve(
     FloatCurve=DataInterface() {propertyValues {Curve {keys [
         <at 0.0018 0 RCIM_Cubic>
         {arriveTangent=0.1593, interpMode=RCIM_Cubic, leaveTangent=0.1593, time=0.2495, value=1}
         {arriveTangent=-1.3324, interpMode=RCIM_Cubic, leaveTangent=-1.3324, time=0.6048, value=0.0961}
         <at 1 0>
        ]}}}
     Scale Curve=0.7
    )
    Scale RGB=&FN_FrontEndVector3Modifier(Gameplay=<5 5 5>, Front End=<2 2 2>)
    ScaleA=true
    ScaleRGB=true
   ) {module=ScaleColor001}
   &ScaleColor(
    Scale Alpha=&FloatFromCurve(
     FloatCurve=DataInterface() {propertyValues {Curve {keys [
         {interpMode=RCIM_Cubic, tangentMode=RCTM_User, time=0, value=0.3}
         {interpMode=RCIM_Cubic, tangentMode=RCTM_User, time=0.5, value=1}
         {tangentMode=RCTM_User, time=1, value=0.3}
        ]}}}
     CurveIndex=Linked() {linkedVariable {name=NPC.FortniteNPC.TimeOfDay_0-1, type=NiagaraFloat}}
    )
    ScaleA=true
    ScaleRGB=false
   ) {module=ScaleColor002}
  ]
  rendererClasses [&NiagaraMeshRendererProperties]
  renderers [{rendererClass=&NiagaraMeshRendererProperties, rendererIndex=0}]
 }
 AirSwipe_R002 {
  emitterUpdateScript [
   &EmitterState(
    Life Cycle Mode=::NewEnumerator1
    Loop Behavior=::NewEnumerator1
    Loop Delay=&FN_FrontEndFloatModifier(Gameplay=1.55, Front End=2.05)
    Loop Duration=1
    UseLoopDelay=true
   )
   &SpawnBurst_Instantaneous(Spawn Count=2, Use Spawn Probability=false, UseLoopCountLimit=false)
  ]
  particleSpawnScript [
   &PartitionParticles()
   &InitializeParticle(
    Color Mode=::NewEnumerator0
    Lifetime=&Lerp_Float(A=0.35, B=0.8, Alpha=&MakeFloatFromInt(Int=$Particles.PartitionParticles.Partition))
    Mesh Scale Mode=::NewEnumerator3
    Mesh Uniform Scale=&Lerp_Float(A=0.6, B=0.8, Alpha=&MakeFloatFromInt(Int=$Particles.PartitionParticles.Partition))
    UsePositionOffset=false
    Write Material Random=true
   )
   &AddVelocity(Velocity=<150 -75 0>)
   &InitialMeshOrientation(
    Mesh Orientation Mode=::NewEnumerator7
    Rotation=&Lerp_Vector(A=<0 0 0.2>, B=<0 0 0>, Alpha=&MakeFloatFromInt(Int=$Particles.PartitionParticles.Partition))
    Use Rotation Vector=true
   )
  ]
  particleUpdateScript [
   &ParticleState()
   &SkeletalMeshLocation(
    Bone / Socket Index=2
    MaskSpawnGroup=false
    Mesh Sampling Type=::NewEnumerator5
    Sampled Position Offset=&FN_FrontEndVector3Modifier(Gameplay=<-60 0 -10>, Front End=<-60 0 -50>)
    Skeletal Mesh=$User.Skeletal Mesh
    Socket Sampling Mode=::NewEnumerator5
   )
   &Drag(Drag=1.5, Use Linear Drag=true, Use Rotational Drag=true)
   &ScaleMeshSize(Scale Factor=&SplitVector(
     Double Component Vector Channels:ECascadeNiagaraTwoVectorChannels=::NewEnumerator0
     Double Component:NiagaraFloat=&FloatFromCurve(
      FloatCurve=DataInterface() {propertyValues {Curve {keys [
          {arriveTangent=0.0236, interpMode=RCIM_Cubic, time=0, value=0.1989}
          {arriveTangent=1.0027, interpMode=RCIM_Cubic, leaveTangent=1.0027, time=0.2593, value=0.8027}
          <at 1 1.2016 RCIM_Cubic>
         ]}}}
      Scale Curve=0.7
     )
     Single Component:NiagaraFloat=&FloatFromCurve(
      FloatCurve=DataInterface() {propertyValues {Curve {keys [<at 0 1 RCIM_Cubic>, {arriveTangent=-1, interpMode=RCIM_Cubic, leaveTangent=-1, time=0.2593, value=1}, <at 1 0 RCIM_Cubic>]}}}
      Scale Curve=2
     )
    ))
   &SolveForcesAndVelocity(Clamp Velocity=false, Limit Acceleration=false)
   &DynamicMaterialParameters(
    Index 0 Param 1=&FloatFromCurve(FloatCurve=DataInterface() {propertyValues {Curve {keys [
         <at 0 -1.9537 RCIM_Cubic>
         {arriveTangent=0.4535, interpMode=RCIM_Cubic, leaveTangent=0.4535, time=0.413, value=-1.7083}
         {interpMode=RCIM_Cubic, leaveTangent=0.1229, time=1, value=-1.5002}
        ]}}})
    Index 0 Param 3=&FloatFromCurve(FloatCurve=DataInterface() {propertyValues {Curve {keys [{arriveTangent=-2.0245, interpMode=RCIM_Cubic, time=0, value=0.004}, <at 1 -0.8019 RCIM_Cubic>]}}})
    Index 0 Param 4=1
    Module.Index0Param0Write=true
    Module.Index0Param1Write=false
    Module.Index0Param2Write=true
    Module.Index0Param3Write=true
   )
   &ScaleColor(
    Scale Alpha=&FloatFromCurve(
     FloatCurve=DataInterface() {propertyValues {Curve {keys [
         <at 0.0018 0 RCIM_Cubic>
         {arriveTangent=0.1593, interpMode=RCIM_Cubic, leaveTangent=0.1593, time=0.2495, value=1}
         {arriveTangent=-1.3324, interpMode=RCIM_Cubic, leaveTangent=-1.3324, time=0.6048, value=0.0961}
         <at 1 0>
        ]}}}
     Scale Curve=0.7
    )
    Scale RGB=&FN_FrontEndVector3Modifier(Gameplay=<5 5 5>, Front End=<2 2 2>)
    ScaleA=true
    ScaleRGB=true
   ) {module=ScaleColor001}
   &ScaleColor(
    Scale Alpha=&FloatFromCurve(
     FloatCurve=DataInterface() {propertyValues {Curve {keys [
         {interpMode=RCIM_Cubic, tangentMode=RCTM_User, time=0, value=0.3}
         {interpMode=RCIM_Cubic, tangentMode=RCTM_User, time=0.5, value=1}
         {tangentMode=RCTM_User, time=1, value=0.3}
        ]}}}
     CurveIndex=Linked() {linkedVariable {name=NPC.FortniteNPC.TimeOfDay_0-1, type=NiagaraFloat}}
    )
    ScaleA=true
    ScaleRGB=false
   ) {module=ScaleColor002}
  ]
  rendererClasses [&NiagaraMeshRendererProperties]
  renderers [{rendererClass=&NiagaraMeshRendererProperties, rendererIndex=0}]
 }
}
summary {
 systemName=NS_Glider_RankedRemixAsparagus_Deploy
 emitters [
  {emitterName=AirSwipe_R001, rendererClasses [&NiagaraMeshRendererProperties]}
  {emitterName=AirSwipe_R002, rendererClasses [&NiagaraMeshRendererProperties]}
 ]
 userVariables [
  User.bIsFrontEnd {type=NiagaraBool}
  User.bIsFrontEndPreview {type=NiagaraBool}
  User.Skeletal Mesh {
   type=NiagaraDataInterfaceSkeletalMesh
   defaultValue {
    struct=NiagaraExt_VariableValue_DataInterface
    value {dataInterface=&self:NiagaraDataInterfaceSkeletalMesh_0}
   }
  }
 ]
}
